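# Physics asset for the front-end mesh of the male Desert Shadow Blade pickaxe, giving its hanging tassel a simulated rig.
# One long capsule covers the two-handed handle, and small capsules cover the tassel core and six chained tassel bones, ending in a fatter tip.
# Swing- and twist-limited constraints link each segment to its parent back to the handle, with tighter twist near the root.
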PhysicsAsset /BRCosmetics/Weapons/FORT_Melee/Pickaxe_Desert_Shadow_Blade_Male/Meshes/Desert_Shadow_Blade_Male_Axe_FrontEnd_PhysicsAsset
bodies {
 Melee_TwoHanded_Handle [{
   shapeType=Capsule
   length=57.8416
   radius=3.775
   shapeName=Melee_TwoHanded_Handle_capsule
   center=<0.0007 0.0594 15.8877>
   rotation=<rot 0 90 0>
  }]
 dyn_Tassle_Core [{
   shapeType=Capsule
   length=0.4117
   radius=0.5
   shapeName=dyn_Tassle_Core_capsule
   center=<0.5161 0.0164 0.0022>
   rotation=<rot 0 -89.8074 -89.7682>
  }]
 dyn_tassle_00 [{shapeType=Capsule, length=1.7701, radius=1, shapeName=dyn_tassle_00_capsule, center=<2 0 0>, rotation=<rot -90 90 -89.9998>}]
 dyn_tassle_01 [{shapeType=Capsule, length=1.0399, radius=1, shapeName=dyn_tassle_01_capsule, center=<1.6766 0 0>, rotation=<rot -90 160.5287 -160.5285>}]
 dyn_tassle_02 [{shapeType=Capsule, length=1.04, radius=1, shapeName=dyn_tassle_02_capsule, center=<1.672 0 0>, rotation=<rot -90 90 -89.9998>}]
 dyn_tassle_03 [{shapeType=Capsule, length=0.78, radius=1, shapeName=dyn_tassle_03_capsule, center=<1.5244 0 0>, rotation=<rot -90 160.5287 -160.5285>}]
 dyn_tassle_04 [{shapeType=Capsule, length=0.2899, radius=1.0301, shapeName=dyn_tassle_04_capsule, center=<1.1457 0 0>, rotation=<rot -90 160.5287 -160.5285>}]
 dyn_tassle_05 [{shapeType=Capsule, length=1.1399, radius=3.1801, shapeName=dyn_tassle_05_capsule, center=<3.7201 0 0>, rotation=<rot -90 90 -89.9998>}]
}
constraints [
 {bone1Name=dyn_tassle_00, bone2Name=dyn_Tassle_Core, swing1LimitDegrees=50, swing2LimitDegrees=30, twistLimitDegrees=0}
 {bone1Name=dyn_tassle_01, bone2Name=dyn_tassle_00, swing1LimitDegrees=50, swing2LimitDegrees=50, twistLimitDegrees=10}
 {bone1Name=dyn_tassle_02, bone2Name=dyn_tassle_01, swing1LimitDegrees=45, swing2LimitDegrees=45, twistLimitDegrees=10}
 {bone1Name=dyn_tassle_03, bone2Name=dyn_tassle_02, swing1LimitDegrees=45, swing2LimitDegrees=45, twistLimitDegrees=10}
 {bone1Name=dyn_tassle_04, bone2Name=dyn_tassle_03, swing1LimitDegrees=45, swing2LimitDegrees=45, twistLimitDegrees=10}
 {bone1Name=dyn_tassle_05, bone2Name=dyn_tassle_04, swing1LimitDegrees=45, swing2LimitDegrees=45, twistLimitDegrees=10}
 {bone1Name=dyn_Tassle_Core, bone2Name=Melee_TwoHanded_Handle, swing1LimitDegrees=50, swing2LimitDegrees=15, twistLimitDegrees=1}
]
